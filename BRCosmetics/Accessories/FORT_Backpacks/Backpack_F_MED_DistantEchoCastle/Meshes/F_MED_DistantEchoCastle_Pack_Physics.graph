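# Physics asset for the DistantEchoCastle backpack cosmetic, simulating a pair of dangling legs with feet on the pack.
# Capsule bodies cover the pack, both legs and both feet, with each foot constrained to its leg and each leg to the pack.
# Every joint locks twist and one swing axis, leaving a single limited hinge-like swing, wider at the legs than at the feet.
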PhysicsAsset /BRCosmetics/Accessories/FORT_Backpacks/Backpack_F_MED_DistantEchoCastle/Meshes/F_MED_DistantEchoCastle_Pack_Physics
bodies {
 dyn_L_foot [{shapeType=Capsule, length=4.142, radius=1.3167, shapeName=dyn_L_foot_capsule, center=<1 0 0>}]
 dyn_L_leg [{shapeType=Capsule, length=6.6478, radius=1.2895, shapeName=dyn_L_leg_capsule, center=<9.4755 0 0>, rotation=<rot 90 0 0>}]
 dyn_R_foot [{shapeType=Capsule, length=4.142, radius=1.3167, shapeName=dyn_R_foot_capsule, center=<1 0 0>}]
 dyn_R_leg [{shapeType=Capsule, length=6.8273, radius=1.4984, shapeName=dyn_R_leg_capsule, center=<10 0 0>, rotation=<rot 90 0 0>}]
 pack [{shapeType=Capsule, length=1, radius=1, shapeName=pack_capsule, center=<0 -6.7247 -3.9988>, rotation=<rot 0 0 -5.7825>}]
}
constraints [
 {bone1Name=dyn_L_leg, bone2Name=pack, swing1LimitDegrees=0, swing2LimitDegrees=25, twistLimitDegrees=0}
 {bone1Name=dyn_L_foot, bone2Name=dyn_L_leg, swing1LimitDegrees=0, swing2LimitDegrees=10, twistLimitDegrees=0}
 {bone1Name=dyn_R_leg, bone2Name=pack, swing1LimitDegrees=0, swing2LimitDegrees=25, twistLimitDegrees=0}
 {bone1Name=dyn_R_foot, bone2Name=dyn_R_leg, swing1LimitDegrees=0, swing2LimitDegrees=10, twistLimitDegrees=0}
]
