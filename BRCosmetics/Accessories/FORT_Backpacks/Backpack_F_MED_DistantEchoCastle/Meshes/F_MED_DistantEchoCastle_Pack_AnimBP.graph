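# Animation Blueprint for the DistantEchoCastle backpack cosmetic, adding physics sway and procedural head motion.
# A RigidBody node in root bone space simulates the pack under a gravity override driven by a character-part state function.
# Additive bone edits then rotate the head, offset legs and root, and tilt the pack while skydiving.
# Head yaw glances side to side using floored-time sines smoothed by FInterpTo, without animation assets.
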AnimBlueprint /BRCosmetics/Accessories/FORT_Backpacks/Backpack_F_MED_DistantEchoCastle/Meshes/F_MED_DistantEchoCastle_Pack_AnimBP
target_skeleton=&/F_MED_DistantEchoCastle_Pack_Skeleton
graphs [
 AnimGraph {nodes [
   n1 = RigidBody!(ComponentPose=Local To Component(LocalPose=Input Pose()), Alpha, OverrideWorldGravity:Vector=Get GravityOverride(), ExternalForce) -> (Pose) {
    bOverrideWorldGravity=true
    overridePhysicsAsset=&/F_MED_DistantEchoCastle_Pack_Physics
    simulationSpace=BaseBoneSpace
    alphaBoolBlend {blendOption=Linear}
    baseBoneRef {boneName=root}
    componentAppliedLinearAccClamp=<3000 3000 3000>
    componentLinearAccScale=<0.25 0.25 0.25>
    componentLinearVelScale=<0.25 0.25 0.25>
    simSpaceSettings {
     maxAngularAcceleration=1000
     maxAngularVelocity=3
     maxLinearAcceleration=700
     maxLinearVelocity=500
     velocityScaleZ=0
     externalLinearDragV=<0.2 0.2 0.2>
    }
   }
   n2 = "Transform (Modify) Bone - Bone: dyn_head"(ComponentPose=@n1, Alpha=1, Translation="0,0,0", Rotation=Get Eye_rotation() -> (Eye_rotation:Rotator), Scale="1,1,1") {
    rotationMode=BMM_Additive
    rotationSpace=BCS_BoneSpace
    alphaBoolBlend {blendOption=Linear}
    alphaScaleBiasClamp {interpSpeedDecreasing=15, interpSpeedIncreasing=15}
    boneToModify {boneName=dyn_head}
    rotation {roll=10}
   }
   n3 = "Transform (Modify) Bone - Bone: dyn_L_leg"(ComponentPose=@n2, Alpha=1, Translation="0,0,0", Rotation="0,0,5", Scale="1,1,1") {
    rotationMode=BMM_Additive
    alphaBoolBlend {blendOption=Linear}
    alphaScaleBiasClamp {interpSpeedDecreasing=15, interpSpeedIncreasing=15}
    boneToModify {boneName=dyn_L_leg}
    rotation {roll=5}
   }
   n4 = "Transform (Modify) Bone - Bone: dyn_R_leg"(ComponentPose=@n3, Alpha=1, Translation="0,0,0", Rotation="0,0,5", Scale="1,1,1") {
    rotationMode=BMM_Additive
    alphaBoolBlend {blendOption=Linear}
    alphaScaleBiasClamp {interpSpeedDecreasing=15, interpSpeedIncreasing=15}
    boneToModify {boneName=dyn_R_leg}
    rotation {roll=5}
   }
   n5 = "Transform (Modify) Bone - Bone: pack"(ComponentPose=@n4, Alpha=Get Is Sky Diving() -> (Is Sky Diving:double), Translation="0,0,0", Rotation="0,0,10") {
    rotationMode=BMM_Additive
    alphaBoolBlend {blendOption=Linear}
    alphaScaleBiasClamp {bInterpResult=true, interpSpeedDecreasing=15, interpSpeedIncreasing=15}
    boneToModify {boneName=pack}
    rotation {roll=10}
   }
   n6 = "Transform (Modify) Bone - Bone: root"(ComponentPose=@n5, Translation="0,-2,0", Rotation="0,0,0") {
    translationMode=BMM_Additive
    translationSpace=BCS_BoneSpace
    alphaBoolBlend {blendOption=Linear}
    boneToModify {boneName=root}
    translation {y=-2}
   }
   Output Pose(Result=Component To Local(ComponentPose=@n6)) {layerGroup=None}
  ]}
 EventGraph {nodes [
   n1 = Event BlueprintUpdateAnimation()
   > n2 = Gravity Override State Function DEPRECATED(CharacterPartAnimInstance=Self-Reference(), GameTimeInSeconds=GetGameTimeInSeconds(), GravityOverrideParams=Get Gravity Override Params())
   > Set GravityOverride(GravityOverride=@n2)
   > Set Is Sky Diving(Is Sky Diving:double="To Float (Boolean)"(InBool=Get bIsSkyDiving())) -> (Output_Get:double)
   n3 = float * float(A=GetGameTimeInSeconds(), B=0.3) {is=title K2Node_PromotableOperator}
   n4 = float * float(A=@n3, B=-2)
   n5 = "To Float (Integer)"(InInt=Floor(A=float + float(A=@n4, B=-0.8)) {is=title K2Node_CallFunction})
   n6 = "To Float (Integer)"(InInt=Floor(A=float * float(A=@n3, B=2)))
   n7 = float - float(A="Clamp (Float)"(Value="Sin (Radians)"(A=@n6), Min=0, Max=1), B="Clamp (Float)"(Value="Sin (Radians)"(A=@n5), Min=0, Max=1))
   > Set Rotation_X_interpolate(Rotation_X_interpolate:double=FInterpTo(
     Current=Get Rotation_X_interpolate() -> (Rotation_X_interpolate:double)
     Target=float * float(A=@n7, B=30)
     DeltaTime=@n1.DeltaTimeX
     InterpSpeed=2
    )) -> (Output_Get:double)
   > Set Eye_rotation(Eye_rotation:Rotator=MakeRotator(Yaw=Get Rotation_X_interpolate())) -> (Output_Get:Rotator)
   TryGetPawnOwner()
  ]}
]
variables [
 Eye_rotation {type=Rotator}
 Gravity Override Params {
  type=GravityOverrideParamsStruct
  default {onGround_Standing {gravityOverride=<0 0 -600>}, skydive_FloatingDown {gravityOverride=<0 -400 150>}}
 }
 GravityOverride {type=Vector}
 Is In Front End {type=boolean}
 Is Sky Diving {type=number}
 Rotation_X_interpolate {type=number}
]
